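# Additive, unlit, two-sided mesh-particle material for the bullet effect of Outlander's Anti-Material Charge gadget.
# Particle color feeds emissive through a shared FX adjustment function.
# Opacity combines panning shield caustics and Fresnel terms with particle alpha, vertex red, an inverted UV1 sphere mask and a dynamic Opacity parameter, then depth-fades.
# A Fresnel term scaled by vertex red and dynamic RefractStrength blends index-of-refraction distortion.
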Material /Game/CharClasses/Outlander/Gadgets/AntiMaterialCharge/Materials/M_AntimaterialChargeBullet
graph {
 properties {
  bUsedWithMeshParticles=true
  bUsedWithNiagaraMeshParticles=true
  blendMode=BLEND_Additive
  pixelDepthOffsetMode=PDOM_Legacy
  refractionMethod=RM_IndexOfRefraction
  shadingModel=MSM_Unlit
  twoSided=true
  usedShadingModels=Unlit
 }
 nodes [
  n1 = ParticleColor()
  n2 = TextureCoordinate() {uTiling=0.5}
  n3 = Add(
   A=TextureSample(UVs=Panner(Coordinate=@n2) {speedX=-2.5, speedY=0.3}) {texture=&T_ShieldCaustics}.RGB
   B=MaterialFunctionCall(Exponent=Constant() {r=4}) {materialFunction=&MF_Fresnel_FX}.Fresnel Inverted
  )
  n4 = SphereMask(A=TextureCoordinate() {coordinateIndex=1}, B=Constant() {r=0.5}, Radius=Constant() {r=0.15}, Hardness=Constant() {r=0.2})
  n5 = Multiply(A=Multiply(A=Multiply(A=@n1.A, B=@n3), B=VertexColor().R), B=OneMinus(None=Power(Base=@n4)))
  n6 = MaterialFunctionCall(Brightness=Constant() {r=1.8}, Exponent=Constant() {r=16}) {materialFunction=&MF_Fresnel_FX}
  DepthFade(
   Opacity=Multiply(
    A=Multiply(A=@n5, B=@n6.Fresnel)
    B=DynamicParameter() {defaultValue=<rgba 0 1 1 1>, paramNames [RefractStrength, Opacity, Param3, Param4]}.Opacity
   )
   FadeDistance=Constant() {r=12}
  )
  n7 = MaterialFunctionCall(
   Brightness=ScalarParameter() {defaultValue=1.05, parameterName=Fresnel Brightness_1}
   Exponent=ScalarParameter() {defaultValue=1.15, parameterName=Fresnel Exponent_1}
  ) {materialFunction=&MF_Fresnel_FX}
  n8 = DynamicParameter() {defaultValue=<rgba 0 1 1 1>, paramNames [RefractStrength, Opacity, Param3, Param4]}
  LinearInterpolate(
   B=ScalarParameter() {defaultValue=0.96, parameterName=Refraction}
   Alpha=Multiply(A=Multiply(A=@n7.Fresnel, B=@n8.RefractStrength), B=VertexColor().R)
  ) {constA=1}
  MaterialFunctionCall(Emissive=@n1.RGB) {materialFunction=&MF_FX_EmissiveAdjustments}
 ]
}
