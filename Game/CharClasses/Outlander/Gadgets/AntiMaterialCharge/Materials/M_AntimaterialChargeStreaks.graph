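# Translucent, unlit, two-sided mesh-particle material for the streak effects of the Outlander's Antimaterial Charge gadget.
# Particle color drives emissive through an FX adjustment function, while opacity multiplies particle alpha with a panning streak texture.
# Two opposing UV gradients, offset by MaskOffset and ShowOffset dynamic parameters, let particles wipe streaks in and out.
# Powered vertex red and a Fresnel term further shape opacity, and depth fade softens intersections.
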Material /Game/CharClasses/Outlander/Gadgets/AntiMaterialCharge/Materials/M_AntimaterialChargeStreaks
graph {
 properties {
  bUsedWithMeshParticles=true
  blendMode=BLEND_Translucent
  pixelDepthOffsetMode=PDOM_Legacy
  previewMesh=/Game/CharClasses/Outlander/Gadgets/AntiMaterialCharge/Meshes/SM_AntimaterialCharge_Streakshape
  shadingModel=MSM_Unlit
  twoSided=true
  usedShadingModels=Unlit
 }
 nodes [
  n1 = ParticleColor()
  n2 = Panner(Coordinate=TextureCoordinate()) {speedX=-2.5}
  n3 = ComponentMask(None=TextureCoordinate() {uTiling=0.5, vTiling=2}) {r=true}
  n4 = DynamicParameter() {defaultValue=<rgba 0 0 1 1>, paramNames [MaskOffset,ShowOffset,Param3,Param4]}
  n5 = Clamp(None=Multiply(A=Add(A=@n3, B=@n4.MaskOffset) {constB=-0.4}) {constB=4})
  n6 = OneMinus(None=ComponentMask(None=TextureCoordinate() {uTiling=0.5, vTiling=2}) {r=true})
  n7 = DynamicParameter() {defaultValue=<rgba 0 0 1 1>, paramNames [MaskOffset,ShowOffset,Param3,Param4]}
  n8 = Clamp(None=Multiply(A=Add(A=@n6, B=@n7.ShowOffset) {constB=0}) {constB=4})
  n9 = Multiply(
   A=@n1.A
   B=Multiply(
    A=TextureSample(UVs=@n2) {texture=/Game/CharClasses/Outlander/Gadgets/AntiMaterialCharge/Textures/T_AntimaterialCharge_Streaks_02}.RGB
    B=Multiply(A=@n5, B=@n8)
   )
  )
  n10 = Power(Base=VertexColor().R) {constExponent=3}
  n11 = Constant() {r=2.5}
  n12 = Constant() {r=8}
  DepthFade(
   Opacity=Multiply(
    A=Multiply(A=@n9, B=@n10)
    B=MaterialFunctionCall(Brightness=@n11, Exponent=@n12) {materialFunction=&MF_Fresnel_FX}.Fresnel
   )
   FadeDistance=Constant() {r=12}
  )
  MaterialFunctionCall(Emissive=@n1.RGB) {materialFunction=&MF_FX_EmissiveAdjustments}
 ]
}
