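# Unlit, two-sided translucent mesh-particle material for the ground dust plane effect of Outlander's AntiMaterialCharge gadget.
# Particle color RGB feeds a shared FX emissive-adjustment function.
# A U-tiled streak texture pans quickly along U, shaped by opposing power curves of U into a band that fades at both ends.
# That mask is multiplied by depth fade, softening intersections, and by particle alpha.
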Material /Game/CharClasses/Outlander/Gadgets/AntiMaterialCharge/Materials/M_AntimaterialCharge_GroundDustPlane
graph {
 properties {
  bUsedWithMeshParticles=true
  blendMode=BLEND_Translucent
  pixelDepthOffsetMode=PDOM_Legacy
  shadingModel=MSM_Unlit
  twoSided=true
  usedShadingModels=Unlit
 }
 nodes [
  n1 = ParticleColor()
  MaterialFunctionCall(Emissive=@n1.RGB) {materialFunction=&MF_FX_EmissiveAdjustments}
  n2 = TextureSample(UVs=Panner(Coordinate=TextureCoordinate() {uTiling=3}) {speedX=-3}) {texture=/Game/CharClasses/Outlander/Gadgets/AntiMaterialCharge/Textures/T_AntimaterialCharge_GroundStreak_02}
  n3 = OneMinus(None=ComponentMask(None=TextureCoordinate()) {r=true})
  n4 = ComponentMask(None=TextureCoordinate()) {r=true}
  n5 = Multiply(A=OneMinus(None=Power(Base=@n3) {constExponent=4}), B=Power(Base=OneMinus(None=@n4)) {constExponent=4})
  n6 = Constant() {r=10}
  Multiply(A=@n1.A, B=Multiply(A=Multiply(A=@n2.RGB, B=@n5), B=DepthFade(FadeDistance=@n6)))
 ]
}
