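# Unlit mesh-particle material for the thruster effect on the Outlander's Anti-Material Charge gadget.
# A mask built from a powered V gradient, inverted Fresnel and two panning sparkle-noise samples blends two warm colors.
# Rounding that mask hardens its upper half to solid Color A before emissive adjustments.
# Two time-driven sines of different periods wiggle vertices in local space, fading out as the powered V gradient rises.
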Material /Game/CharClasses/Outlander/Gadgets/AntiMaterialCharge/Materials/M_Outlander_AntiMaterialChargeThruster_01
graph {
 properties {
  bUsedWithMeshParticles=true
  bUsedWithNiagaraMeshParticles=true
  pixelDepthOffsetMode=PDOM_Legacy
  shadingModel=MSM_Unlit
  usedShadingModels=Unlit
 }
 nodes [
  n1 = VectorParameter() {parameterName=Color B, defaultValue=<rgba 1.5 0.3944 0.1202 0>}
  n2 = VectorParameter() {parameterName=Color A, defaultValue=<rgba 7 4.1946 1.3214 0>}
  n3 = ComponentMask(None=TextureCoordinate()) {g=true}
  n4 = Power(Base=@n3, Exp=ScalarParameter() {defaultValue=10, parameterName=LERP Power})
  n5 = MaterialFunctionCall(
   Brightness=ScalarParameter() {defaultValue=1, parameterName=Fresnel Brightness}
   Exponent=ScalarParameter() {defaultValue=3, parameterName=Fresnel Exponent}
  ) {materialFunction=&MF_Fresnel_FX}
  n6 = TextureCoordinate() {uTiling=1.5, vTiling=0.15}
  n7 = TextureCoordinate() {uTiling=2, vTiling=0.116}
  n8 = Add(
   A=TextureSample(UVs=Panner(Coordinate=@n6) {bFractionalPart=true, speedX=1.5, speedY=1.744}) {samplerType=SAMPLERTYPE_LinearColor, texture=/Game/Gadgets/Assets/Slow_Field/Textures/T_Slowfield_SparkleNoise_01}.R
   B=TextureSample(UVs=Panner(Coordinate=@n7) {bFractionalPart=true, speedX=-0.4, speedY=1.92}) {samplerType=SAMPLERTYPE_LinearColor, texture=&T_Slowfield_SparkleNoise_01}.G
  )
  n9 = Clamp(None=Add(A=Add(A=@n4, B=@n5.Fresnel Inverted), B=Multiply(A=@n3, B=@n8)))
  n10 = Round(None=@n9)
  MaterialFunctionCall(Emissive=LinearInterpolate(A=LinearInterpolate(A=@n1.RGB, B=@n2.RGB, Alpha=@n9), B=LinearInterpolate(A=@n1.RGB, B=@n2.RGB, Alpha=@n10), Alpha=@n10)) {materialFunction=&MF_FX_EmissiveAdjustments}
  n11 = OneMinus(None=@n4)
  n12 = Multiply(A=Time(), B=ScalarParameter() {defaultValue=9, parameterName=WiggleSpeed})
  n13 = Sine(None=@n12) {period=1.189}
  n14 = Sine(None=@n12) {period=0.782}
  n15 = VectorParameter() {parameterName=WPO Intensity, defaultValue=<rgba 1.2 0 0 1>}
  n16 = MaterialFunctionCall(R=Add(A=Add(A=@n13, B=@n14), B=@n15.A)) {materialFunction=&AppendMany}
  Transform(None=Multiply(A=Multiply(A=@n11, B=@n16.RGB), B=@n15.RGB)) {transformSourceType=TRANSFORMSOURCE_Local}
 ]
}
